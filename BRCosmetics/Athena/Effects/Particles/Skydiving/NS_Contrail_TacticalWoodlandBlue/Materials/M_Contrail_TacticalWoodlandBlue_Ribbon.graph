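# Unlit alpha-composite Niagara ribbon material for the Tactical Woodland Blue skydiving contrail, drawing hue-shifted animated lines over a soft background tint.
# Distorted, skewed UVs pan by time plus a particle offset, shifting horizontally on each wrap.
# Two line-texture samples at different scales use explicit derivatives around Frac, then are thresholded and combined.
# UV masks shape the result, while particle alpha and depth fade scale emissive and opacity.
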Material /BRCosmetics/Athena/Effects/Particles/Skydiving/NS_Contrail_TacticalWoodlandBlue/Materials/M_Contrail_TacticalWoodlandBlue_Ribbon
graph {
 properties {
  bUsedWithNiagaraRibbons=true
  blendMode=BLEND_AlphaComposite
  pixelDepthOffsetMode=PDOM_Legacy
  shadingModel=MSM_Unlit
  usedShadingModels=Unlit
 }
 nodes [
  n1 = VectorParameter() {parameterName=Background Color, defaultValue=<rgba 0 0.0483 0.1 0>}
  n2 = MaterialFunctionCall(Float2=TextureCoordinate()) {materialFunction=&BreakOutFloat2Components}
  n3 = AppendVector(A=@n2.G, B=@n2.G)
  n4 = Subtract(A=ComponentMask(None=@n3) {r=true}) {constB=0.5}
  n5 = ComponentMask(None=@n3) {g=true}
  n6 = ScalarParameter() {defaultValue=0.25, parameterName=UV Distortion Offset}
  n7 = Saturate(None=Divide(A=Subtract(A=@n5, B=@n6), B=ScalarParameter() {defaultValue=0.1, parameterName=UV Distortion Size}))
  n8 = ScalarParameter() {defaultValue=0.5, parameterName=UV Distortion Intensity, sliderMax=1}
  n9 = Multiply(A=@n4, B=Add(A=Multiply(A=@n7, B=@n8)))
  n10 = Saturate(None=Add(A=@n9) {constB=0.5})
  n11 = Multiply(A=@n10, B=OneMinus(None=@n10))
  n12 = ScalarParameter() {defaultValue=0.5, parameterName=Color Background Softness A}
  n13 = Saturate(None=Divide(A=@n11, B=ScalarParameter() {defaultValue=0.1, parameterName=Color Background Softness B}) {constB=0.02})
  n14 = Add(A=Saturate(None=Divide(A=@n11, B=@n12)), B=@n13)
  n15 = OneMinus(None=@n5)
  n16 = TextureCoordinate()
  n17 = Saturate(None=Divide(A=ComponentMask(None=@n16) {r=true}) {constB=0.1})
  n18 = Multiply(A=@n17, B=@n17)
  n19 = Multiply(A=Multiply(A=Divide(A=@n14), B=Multiply(A=@n15, B=@n15)), B=@n18)
  n20 = ConstantBiasScale(None=ComponentMask(None=TextureCoordinate()) {r=true}) {bias=-0.5, scale=2}
  n21 = VectorParameter() {parameterName="Lines Color && Hue Shift", defaultValue=<rgba 0 0.6667 1 0.11>}
  n22 = Multiply(A=OneMinus(None=Abs(None=@n20)), B=@n21.A)
  n23 = Abs(None=@n4)
  n24 = Multiply(A=@n23, B=ScalarParameter() {defaultValue=0.2, parameterName=UV Length Skew})
  n25 = AppendVector(
   A=Multiply(A=@n9, B=ScalarParameter() {defaultValue=0.2, parameterName=UV Scale Width})
   B=Multiply(A=Add(A=@n5, B=@n24), B=ScalarParameter() {defaultValue=0.72, parameterName=UV Scale Length})
  )
  n26 = Time()
  n27 = ScalarParameter() {defaultValue=3, parameterName=Time Scale}
  n28 = Panner(
   Coordinate=@n25
   Time=Add(A=Multiply(A=@n26, B=@n27), B=DynamicParameter() {paramNames [Time Offset, - unused -, - unused -, - unused -]}.Time Offset)
  ) {speedY=-1}
  n29 = Multiply(A=Floor(None=ComponentMask(None=@n28) {g=true})) {constB=0.135}
  n30 = Add(A=@n28, B=AppendVector(A=Frac(None=@n29), B=Constant()))
  n31 = DDX(Value=@n25)
  n32 = DDY(Value=@n25)
  n33 = TextureSample!(UVs=Frac(None=@n30), Tex, "DDX(UVs)"=@n31, "DDY(UVs)"=@n32, Apply View MipBias) {
   mipValueMode=TMVM_Derivative
   samplerType=SAMPLERTYPE_LinearGrayscale
   texture=/BRCosmetics/Gadgets/Assets/VinderTech_GliderChute/Glider_Tactical_Woodland_Blue_Male/FX/Textures/T_Glider_TacticalWoodlandBlue_Lines
  }
  n34 = Divide(A=OneMinus(None=@n5)) {constB=0.4}
  n35 = Multiply(A=@n5, B=@n5)
  n36 = Saturate(None=Divide(A=Multiply(A=@n35, B=@n23)) {constB=0.4})
  n37 = OneMinus(None=Multiply(A=Saturate(None=@n34), B=OneMinus(None=@n36)))
  n38 = ScalarParameter() {defaultValue=0.1, parameterName=Softness}
  n39 = Saturate(None=Divide(A=Subtract(A=@n33.R, B=@n37), B=@n38))
  n40 = Multiply(A=@n30) {constB=2}
  n41 = TextureSample(UVs=Frac(None=Add(A=@n40) {constB=0.5}), "DDX(UVs)"=@n31, "DDY(UVs)"=@n32) {
   mipValueMode=TMVM_Derivative
   samplerType=SAMPLERTYPE_LinearGrayscale
   texture=&T_Glider_TacticalWoodlandBlue_Lines
  }
  n42 = Subtract(A=@n41.R, B=@n37)
  n43 = Multiply(A=Saturate(None=Divide(A=@n42, B=@n38)), B=ScalarParameter() {defaultValue=0.1, parameterName=Frequency Blend, sliderMax=1})
  n44 = Multiply(A=Saturate(None=Add(A=@n39, B=@n43)), B=Multiply(A=@n18, B=@n13))
  n45 = Divide(
   A=LinearInterpolate(
    A=Multiply(A=@n1.RGB, B=@n19)
    B=MaterialFunctionCall(Hue Shift Percentage=@n22, Texture=@n21.RGB) {materialFunction=&HueShift}.Result
    Alpha=@n44
   )
   B=ComponentMask(None=TextureCoordinate()) {r=true}
  )
  n46 = DepthFade(Opacity=ParticleColor().A, FadeDistance=ScalarParameter() {defaultValue=111, parameterName=Depth Fade Distance})
  MaterialFunctionCall(Emissive=Multiply(A=Clamp(None=@n45) {clampMode=CMODE_ClampMax, maxDefault=10}, B=@n46)) {materialFunction=&MF_FX_EmissiveAdjustments}
  Multiply(A=@n44, B=@n46)
 ]
}
